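# Niagara effect for the Pirate 02 back bling, converted from Cascade, showing occasional glowing red-pink sprites on its two knives.
# Two near-identical sprite emitters follow the right and left dyn_knife bones, resampling skeletal mesh position each update.
# A sequential spawn index alternates between two local offsets per knife.
# Sprites rotate slowly, pulse through size and alpha curves, get boosted HDR color, and are offset toward the camera.
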NiagaraSystem /BRCosmetics/Accessories/FORT_Backpacks/Backpack_F_MED_Pirate_02/FX/NS_MED_Pirate_02_Converted
dependencies {
 usedDataInterfaces [&NiagaraDataInterfaceCurve, &NiagaraDataInterfaceSkeletalMesh]
 usedDynamicInputs []
 usedModules [&SystemState]
}
emitters {
 Particle_Emitter {
  emitterUpdateScript [
   &EmitterState(Life Cycle Mode=::NewEnumerator1, Loop Duration=1, UseLoopDelay=true)
   &SpawnRate(SpawnRate=0.5, Use Spawn Probability=false)
  ]
  particleSpawnScript [
   /SetVariables(Particles.SpriteRotationRate=&RandomRangeFloat(Minimum=-0.08, Maximum=0.07, Evaluation Type=::NewEnumerator1)) {isSetParameters=true}
   /SetVariables(Particles.SkeletonSamplingIndex:NiagaraInt32=/CascadeToNiagaraConverter/NiagaraScripts/DynamicInputScripts/SequentialInt(Use Array Data:NiagaraBool=false, First Number:NiagaraInt32=0, Sequence:NiagaraDataInterfaceArrayInt32=?, Last Number:NiagaraInt32=1)) {isSetParameters=true}
   &InitializeParticle(
    Lifetime=&RandomRangeFloat(Minimum=0.75, Maximum=1.45, Evaluation Type=::NewEnumerator1)
    Sprite Rotation Angle=&RandomRangeFloat(Minimum=0, Maximum=1, Evaluation Type=::NewEnumerator1)
    Sprite Rotation Mode=::NewEnumerator1
    Sprite Size=&Vector2DFromFloat(Value=&RandomRangeFloat(Minimum=9.5, Maximum=14.5, Evaluation Type=::NewEnumerator1))
    Sprite Size Mode=::NewEnumerator4
    UsePositionOffset=false
    Write Material Random=true
   )
   &Color(Color=&MakeLinearColorFromVectorAndFloat("Vector (RGB)"=<1 0.2772 0.2274>, "Float (Alpha)"=1))
  ]
  particleUpdateScript [
   &ParticleState()
   &ScaleSpriteSize(
    Scale Sprite Size Mode=::NewEnumerator0
    Uniform Scale Factor=&FloatFromCurve(FloatCurve=DataInterface() {propertyValues {Curve {keys [
         {arriveTangent=9.9733, interpMode=RCIM_Cubic, leaveTangent=9.9733, tangentMode=RCTM_User, time=0, value=0}
         {arriveTangent=0.963, interpMode=RCIM_Cubic, leaveTangent=0.963, tangentMode=RCTM_User, time=0.1593, value=0.8516}
         {arriveTangent=0.0823, interpMode=RCIM_Cubic, leaveTangent=0.0823, time=0.5036, value=0.5781}
         {arriveTangent=-0.545, interpMode=RCIM_Cubic, leaveTangent=-0.545, time=0.7974, value=0.9041}
         <at 1 0.3075>
        ]}}})
   )
   &SpriteRotationRate(Rotation Rate=&AngleConversion(
     Angle:NiagaraFloat=$Particles.SpriteRotationRate
     Angle Input:ENiagara_AngleInput=::NewEnumerator1
     Angle Output:ENiagara_AngleInput=::NewEnumerator0
    ))
   &ScaleColor(
    Scale Alpha=&FloatFromCurve(FloatCurve=DataInterface() {propertyValues {Curve {keys [
         {arriveTangent=10.6971, interpMode=RCIM_Cubic, leaveTangent=10.6971, tangentMode=RCTM_User, time=0, value=0}
         {arriveTangent=1.3141, interpMode=RCIM_Cubic, leaveTangent=1.3141, time=0.1724, value=0.9581}
         {arriveTangent=0.0483, interpMode=RCIM_Cubic, leaveTangent=0.0483, time=0.388, value=0.5099}
         {arriveTangent=-0.0572, interpMode=RCIM_Cubic, leaveTangent=-0.0572, tangentMode=RCTM_User, time=0.6387, value=0.9806}
         {arriveTangent=-2.8729, leaveTangent=-2.8729, time=1, value=0}
        ]}}})
    Scale RGB=<30 30 30>
    ScaleA=true
    ScaleRGB=true
   )
   &CameraOffset(Camera Offset Amount=8)
   &SkeletalMeshLocation(
    Bone / Socket Index=$Particles.SkeletonSamplingIndex
    Bone Sampling Mode=::NewEnumerator4
    MaskSpawnGroup=false
    Orientation Sampling=::NewEnumerator1
    Sampled Position Offset=&MergeHelper_VectorSwitch(0=<8.2 0 0>, 1=<10 0 1>, 2=?, 3=?, Index=$Particles.SkeletonSamplingIndex, Num=::NewEnumerator0)
    Skeletal Mesh=DataInterface() {propertyValues {
      DefaultMesh=null
      ExcludeBoneName=root
      OptionalComponentClass=null
      PreviewMesh=null
      SkinningMode=SkinOnTheFly
      SoftSourceActor=null
      SourceMode=Default
      UvSetIndex=0
      WholeMeshLOD=-1
      bExcludeBone=false
      bReadDeformedGeometry=true
      bRequireCurrentFrameData=true
      ComponentTags []
      FilteredBones [dyn_knife_r,dyn_knife_r]
      FilteredSockets []
      MeshUserParameter {parameter {name=None, typeDefHandle {registeredTypeIndex=1}, varData []}}
      SamplingRegions []
     }}
   )
  ]
  rendererClasses [&NiagaraSpriteRendererProperties]
  renderers [{rendererClass=&NiagaraSpriteRendererProperties}]
 }
 Particle_Emitter001 {
  emitterUpdateScript [
   &EmitterState(Life Cycle Mode=::NewEnumerator1, Loop Duration=1, UseLoopDelay=true)
   &SpawnRate(SpawnRate=0.6, Use Spawn Probability=false)
  ]
  particleSpawnScript [
   /SetVariables(Particles.SpriteRotationRate=&RandomRangeFloat(Minimum=-0.08, Maximum=0.07, Evaluation Type=::NewEnumerator1)) {isSetParameters=true}
   /SetVariables(Particles.SkeletonSamplingIndex:NiagaraInt32=&SequentialInt(Use Array Data:NiagaraBool=false, First Number:NiagaraInt32=0, Sequence:NiagaraDataInterfaceArrayInt32=?, Last Number:NiagaraInt32=1)) {isSetParameters=true}
   &InitializeParticle(
    Lifetime=&RandomRangeFloat(Minimum=0.8, Maximum=1.45, Evaluation Type=::NewEnumerator1)
    Sprite Rotation Angle=&RandomRangeFloat(Minimum=0, Maximum=1, Evaluation Type=::NewEnumerator1)
    Sprite Rotation Mode=::NewEnumerator1
    Sprite Size=&Vector2DFromFloat(Value=&RandomRangeFloat(Minimum=9, Maximum=14.5, Evaluation Type=::NewEnumerator1))
    Sprite Size Mode=::NewEnumerator4
    UsePositionOffset=false
    Write Material Random=true
   )
   &Color(Color=&MakeLinearColorFromVectorAndFloat("Vector (RGB)"=<1 0.2772 0.2274>, "Float (Alpha)"=1))
  ]
  particleUpdateScript [
   &ParticleState()
   &ScaleSpriteSize(
    Scale Sprite Size Mode=::NewEnumerator0
    Uniform Scale Factor=&FloatFromCurve(FloatCurve=DataInterface() {propertyValues {Curve {keys [
         {arriveTangent=9.9733, interpMode=RCIM_Cubic, leaveTangent=9.9733, tangentMode=RCTM_User, time=0, value=0}
         {arriveTangent=0.963, interpMode=RCIM_Cubic, leaveTangent=0.963, tangentMode=RCTM_User, time=0.1593, value=0.8516}
         {arriveTangent=0.0823, interpMode=RCIM_Cubic, leaveTangent=0.0823, time=0.5036, value=0.5781}
         {arriveTangent=-0.545, interpMode=RCIM_Cubic, leaveTangent=-0.545, time=0.7974, value=0.9041}
         <at 1 0.3075>
        ]}}})
   )
   &SpriteRotationRate(Rotation Rate=&AngleConversion(
     Angle:NiagaraFloat=$Particles.SpriteRotationRate
     Angle Input:ENiagara_AngleInput=::NewEnumerator1
     Angle Output:ENiagara_AngleInput=::NewEnumerator0
    ))
   &ScaleColor(
    Scale Alpha=&FloatFromCurve(FloatCurve=DataInterface() {propertyValues {Curve {keys [
         {arriveTangent=10.6971, interpMode=RCIM_Cubic, leaveTangent=10.6971, tangentMode=RCTM_User, time=0, value=0}
         {arriveTangent=1.3141, interpMode=RCIM_Cubic, leaveTangent=1.3141, time=0.1724, value=0.9581}
         {arriveTangent=0.0483, interpMode=RCIM_Cubic, leaveTangent=0.0483, time=0.388, value=0.5099}
         {arriveTangent=-0.0572, interpMode=RCIM_Cubic, leaveTangent=-0.0572, tangentMode=RCTM_User, time=0.6387, value=0.9806}
         {arriveTangent=-2.8729, leaveTangent=-2.8729, time=1, value=0}
        ]}}})
    Scale RGB=<30 30 30>
    ScaleA=true
    ScaleRGB=true
   )
   &CameraOffset(Camera Offset Amount=3.55)
   &SkeletalMeshLocation(
    Bone / Socket Index=$Particles.SkeletonSamplingIndex
    Bone Sampling Mode=::NewEnumerator4
    MaskSpawnGroup=false
    Orientation Sampling=::NewEnumerator1
    Sampled Position Offset=&MergeHelper_VectorSwitch(0=<9.2 0 0.95>, 1=<10.5 0 -0.5>, 2=?, 3=?, Index=$Particles.SkeletonSamplingIndex, Num=::NewEnumerator0)
    Skeletal Mesh=DataInterface() {propertyValues {
      DefaultMesh=null
      ExcludeBoneName=root
      OptionalComponentClass=null
      PreviewMesh=null
      SkinningMode=SkinOnTheFly
      SoftSourceActor=null
      SourceMode=Default
      UvSetIndex=0
      WholeMeshLOD=-1
      bExcludeBone=false
      bReadDeformedGeometry=true
      bRequireCurrentFrameData=true
      ComponentTags []
      FilteredBones [dyn_knife_l,dyn_knife_l]
      FilteredSockets []
      MeshUserParameter {parameter {name=None, typeDefHandle {registeredTypeIndex=1}, varData []}}
      SamplingRegions []
     }}
   )
  ]
  rendererClasses [&NiagaraSpriteRendererProperties]
  renderers [{rendererClass=&NiagaraSpriteRendererProperties}]
 }
}
summary {
 systemName=NS_MED_Pirate_02_Converted
 emitters [
  {emitterName=Particle_Emitter, rendererClasses [&NiagaraSpriteRendererProperties]}
  {emitterName=Particle_Emitter001, rendererClasses [&NiagaraSpriteRendererProperties]}
 ]
 userVariables []
}
