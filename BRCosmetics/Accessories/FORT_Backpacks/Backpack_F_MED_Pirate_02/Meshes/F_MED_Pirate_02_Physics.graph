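# Physics asset for the F_MED_Pirate_02 backpack cosmetic, giving physical motion to its dangling map and anchor bones.
# A box body covers the backpack root, while capsule bodies cover the map and the two anchor bones.
# Constraints form a chain from root to map to first anchor to second anchor, each with limited swing and locked twist, with the map's second swing axis much tighter than the anchors'.
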PhysicsAsset /BRCosmetics/Accessories/FORT_Backpacks/Backpack_F_MED_Pirate_02/Meshes/F_MED_Pirate_02_Physics
bodies {
 backpack_root [{shapeType=Box, extentX=27.2188, extentY=4.9431, extentZ=27.2188, center=<-11.4317 -3.1609 -0.0001>, rotation=<rot 0 5.0003 0>}]
 dyn_anchor01 [{shapeType=Capsule, length=2.9728, radius=2.2767, center=<0.9141 -0.3795 0>, rotation=<rot 90 -0.0199 -0.0201>}]
 dyn_anchor02 [{shapeType=Capsule, length=5.1413, radius=1.998, center=<2.6292 -0.5055 0>, rotation=<rot 90 -0.0199 -0.0201>}]
 dyn_map [{shapeType=Capsule, length=6.6601, radius=2.5251, center=<1.4946 -0.0001 0>, rotation=<rot 90 0.0131 0.0129>}]
}
constraints [
 {bone1Name=dyn_map, bone2Name=backpack_root, swing1LimitDegrees=25, swing2LimitDegrees=5, twistLimitDegrees=45, twistMotion=Locked}
 {bone1Name=dyn_anchor01, bone2Name=dyn_map, swing1LimitDegrees=25, swing2LimitDegrees=45, twistLimitDegrees=45, twistMotion=Locked}
 {bone1Name=dyn_anchor02, bone2Name=dyn_anchor01, swing1LimitDegrees=25, swing2LimitDegrees=45, twistLimitDegrees=45, twistMotion=Locked}
]
