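# Animation Blueprint for the Pirate 02 cosmetic backpack, posing its root from the wearer's body motion and adding physics.
# Each update maps owner spine yaw to a root rotation, and left upper-arm yaw to a swing alpha, zeroed when skydiving or custom animations can play.
# Two additive bone transforms apply these to the ref pose before RigidBody simulates with an override physics asset.
# An Anim Dynamics chain for map and anchor bones is unconnected to output.
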AnimBlueprint /BRCosmetics/Accessories/FORT_Backpacks/Backpack_F_MED_Pirate_02/Meshes/F_MED_Pirate_02_AnimBP
target_skeleton=&/F_MED_Pirate_02_Skeleton
graphs [
 AnimGraph {nodes [
   n1 = Anim Dynamics - Bone: dyn_map(Alpha=1, ExternalForce="0,0,0") {
    angularDampingOverride=0.9
    bGravityOverrideInSimSpace=true
    bOverrideAngularDamping=true
    bUseGravityOverride=true
    alphaBoolBlend {blendOption=Linear}
    boundBone {boneName=dyn_map}
    componentAppliedLinearAccClamp=<1000 1000 1000>
    componentLinearAccScale=<1 1 1>
    componentLinearVelScale=<1 1 1>
    gravityOverride {z=-980}
    physicsBodyDefinitions [{
      sphereCollisionRadius=0
      boundBone {boneName=dyn_map}
      boxExtents=<8 15 12>
      constraintSetup {angularLimitsMax=<0 5 0>, angularLimitsMin=<0 -5 -35>, angularTarget=<0 0 0>}
      localJointOffset=<1 0 0>
     }]
   }
   n2 = Anim Dynamics - Bone: dyn_anchor01(ComponentPose=@n1, Alpha=1, ExternalForce="0,0,0") {
    angularDampingOverride=0.9
    bGravityOverrideInSimSpace=true
    bOverrideAngularDamping=true
    bUseGravityOverride=true
    alphaBoolBlend {blendOption=Linear}
    boundBone {boneName=dyn_anchor01}
    componentAppliedLinearAccClamp=<1000 1000 1000>
    componentLinearAccScale=<1 1 1>
    componentLinearVelScale=<1 1 1>
    gravityOverride {z=-980}
    physicsBodyDefinitions [{
      sphereCollisionRadius=0
      boundBone {boneName=dyn_anchor01}
      boxExtents=<8 15 12>
      constraintSetup {angularLimitsMax=<0 45 25>, angularLimitsMin=<0 -45 -35>, angularTarget=<0 0 0>}
      localJointOffset=<1 0 0>
     }]
   }
   Anim Dynamics - Bone: dyn_anchor02(ComponentPose=@n2, Alpha=1, ExternalForce="0,0,0") {
    angularDampingOverride=0.99
    bGravityOverrideInSimSpace=true
    bOverrideAngularDamping=true
    bUseGravityOverride=true
    alphaBoolBlend {blendOption=Linear}
    boundBone {boneName=dyn_anchor02}
    componentAppliedLinearAccClamp=<1000 1000 1000>
    componentLinearAccScale=<1 1 1>
    componentLinearVelScale=<1 1 1>
    gravityOverride {z=-980}
    physicsBodyDefinitions [{
      sphereCollisionRadius=0
      boundBone {boneName=dyn_anchor02}
      boxExtents=<8 15 12>
      constraintSetup {angularLimitsMax=<0 45 25>, angularLimitsMin=<0 -45 -35>, angularTarget=<0 0 0>}
      localJointOffset=<1 0 0>
     }]
   }
   n3 = "Transform (Modify) Bone - Bone: backpack_root"(
    ComponentPose=Local To Component(LocalPose=Local Space Ref Pose())
    Alpha=1
    Translation="0,0,0"
    Rotation=Get backpack_rotate()
    Scale="1,1,1"
   ) {rotationMode=BMM_Additive, rotationSpace=BCS_BoneSpace, alphaBoolBlend {blendOption=Linear}, boneToModify {boneName=backpack_root}}
   n4 = "Transform (Modify) Bone - Bone: backpack_root"(ComponentPose=@n3, Alpha=Get swing_rotate_alpha() -> (swing_rotate_alpha:double), Translation="0,0,0", Rotation="50,0,20", Scale="1,1,1") {
    rotationMode=BMM_Additive
    rotationSpace=BCS_BoneSpace
    alphaBoolBlend {blendOption=Linear}
    alphaScaleBiasClamp {bInterpResult=true, interpSpeedDecreasing=5}
    boneToModify {boneName=backpack_root}
    rotation {pitch=50, roll=20}
   }
   n5 = RigidBody(ComponentPose=@n4) {
    bOverrideWorldGravity=true
    cachedBoundsScale=1
    overridePhysicsAsset=&/F_MED_Pirate_02_Physics
    alphaBoolBlend {blendOption=Linear}
    componentAppliedLinearAccClamp=<1000 1000 1000>
    componentLinearAccScale=<0.8 0.8 0.8>
    componentLinearVelScale=<0.8 0.8 0.8>
    overrideWorldGravity {z=-980}
   }
   Output Pose(Result=Component To Local(ComponentPose=@n5)) {layerGroup=None}
  ]}
 EventGraph {nodes [
   Event BlueprintUpdateAnimation()
   n1 = GetOwnerSkeletalMeshComponent()
   n2 = BreakRotator(InRot=BreakTransform(InTransform=GetDeltaTransformFromRefPose(self=@n1, BoneName=spine_05, BaseName=spine_01)).Rotation)
   > Set backpack_rotate(backpack_rotate=MakeRotator(Yaw=MapRangeClamped(Value=@n2.Yaw, InRangeA=0, InRangeB=-30, OutRangeB=30)))
   n3 = BreakTransform(InTransform=GetDeltaTransformFromRefPose(self=GetOwnerSkeletalMeshComponent(), BoneName=upperarm_twist_01_l, BaseName=clavicle_l))
   > Set swing_rotate_alpha(swing_rotate_alpha:double=MapRangeClamped(Value=BreakRotator(InRot=@n3.Rotation).Yaw, InRangeA=10, InRangeB=-40)) -> (Output_Get:double)
   > Branch(Condition=OR Boolean(A=Get bIsSkyDiving(), B=Get bCanPlayCustomAnimations() -> (bCanPlayCustomAnimations:bool)))
   > Set swing_rotate_alpha(swing_rotate_alpha=0)
  ]}
]
variables [
 backpack_rotate {type=Rotator}
 swing_rotate_alpha {type=number}
]
